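; Super Bridge grenade projectile with helper functions for stamping grid-snapped structures at an impact point; BeginPlay only prints a "shouldn't be possible" message.
; Instigator facing picks a cardinal direction, whose mapped offsets orient cleared cells and spawned pieces.
; ClearAndBuild clears cells, then spawns budget-checked, team-assigned pieces, destroying those at invalid world locations.
; Other helpers batch-clear cells, spawn dynamic builders, sort props and trace-place tires.
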(event EventBeginPlay
  (Development|PrintString "This shouldnt even be possible.... how'd you do this" true true "(R=1.000000,G=0.000000,B=0.088284,A=1.000000)" 8.0))

(fn CalculateCardinalDirection
  (bind _returnvalue (Game|GetInstigator))
  (bind _returnvalue_2 (Transformation|GetActorForwardVector _returnvalue))
  (bind _x (.x _returnvalue_2))
  (bind _y (.y _returnvalue_2))
  (if (>= (Math|Float|Absolute(Float) _x) (Math|Float|Absolute(Float) _y))
    (if (>= _x 0.0)
      (Variables|Orientation|SetCardinalDirection "NewEnumerator0")
      (else
        (Variables|Orientation|SetCardinalDirection "NewEnumerator1")))
    (elif (>= _y 0.0)
      (Variables|Orientation|SetCardinalDirection "NewEnumerator2")
      (else
        (Variables|Orientation|SetCardinalDirection "NewEnumerator3")))))

(fn CallDynamicBuilder
  (params (Class :type Actor%20Class%20Reference) (Location :type Vector))
  (bind _returnvalue (Game|SpawnActorfromClass Class (Math|Transform|MakeTransform Location (Math|Rotator|MakeRotator 0.0 0.0 (Utilities|Map|Find (Variables|Orientation|GetOrientationYawOffsets) (Variables|Orientation|GetCardinalDirection)))) "AdjustIfPossibleButDontSpawnIfColliding" "SelectDefaultAtRuntime" 0 0))
  (bind _asbuilding_actor (Utilities|Casting|CastToBuildingActor _returnvalue))
  (bind _asfort_player_controller (Utilities|Casting|CastToFortPlayerController (Game|GetInstigatorController)))
  (Building|InitializeKismetSpawnedBuildingActor _asbuilding_actor 0 _asfort_player_controller true 0 false))

(fn ClearAndBuild
  (params (ClearedGridCells :type Array%20of%20Vectors) (BGAClassArray :type Array%20of%20Actor%20Class%20References) (BGALocationArray :type Array%20of%20Vectors) (BGARotationArray :type Array%20of%20Rotators) (DestroyFloorsQuickly :type Boolean))
  (for _array_element ClearedGridCells
    (Tower|DestroyActorsinGridCellLocation (Building|GetBuildingStructuralSupportSystem) (+ (Variables|BuildingAndClearing|GetBaseDestination) (Math|Vector|RotateVectorAroundAxis _array_element (Utilities|Map|Find (Variables|Orientation|GetOrientationYawOffsets) (Variables|Orientation|GetCardinalDirection)) "0, 0,1.000000")) DestroyFloorsQuickly))
  (CallFunction|CreateBaseSection (Variables|BuildingAndClearing|GetBaseDestination) BGAClassArray BGALocationArray BGARotationArray))

(fn ClearOnly
  (params (ClearedGridCells :type Array%20of%20Vectors) (DestroyFloorsQuickly :type Boolean))
  (bind _buildingstructuralsupportsystem (Building|GetBuildingStructuralSupportSystem))
  (Building|StartActorRemovalBatch _buildingstructuralsupportsystem)
  (for _array_element ClearedGridCells
    (Tower|DestroyActorsinGridCellLocation (Building|GetBuildingStructuralSupportSystem) (+ (Variables|BuildingAndClearing|GetBaseDestination) (Math|Vector|RotateVectorAroundAxis _array_element (Utilities|Map|Find (Variables|Orientation|GetOrientationYawOffsets) (Variables|Orientation|GetCardinalDirection)) "0, 0,1.000000")) DestroyFloorsQuickly))
  (Building|StopActorRemovalBatch _buildingstructuralsupportsystem))

(fn CreateBaseSection
  (params (ReferenceLocation :type Vector) (BGAClassArray :type Array%20of%20Actor%20Class%20References) (BGALocationArray :type Array%20of%20Vectors) (BGARotationArray :type Array%20of%20Rotators))
  (if (> (Utilities|Array|Length BGAClassArray) 0)
    (if (Creative|Budget|CanPlaceClassInstances (Fortnite|GetFortPlayerControllerfromActor (Actor|GetOwner)) BGAClassArray ReferenceLocation)
      (for _index (range (Utilities|Array|Length BGAClassArray))
        (bind _returnvalue (Game|SpawnActorfromClass (Utilities|Array|Get(acopy) BGAClassArray _index) (Math|Transform|MakeTransform (Math|Vector|vector+vector ReferenceLocation (Math|Vector|RotateVectorAroundAxis (Utilities|Array|Get(aref) BGALocationArray _index) (Utilities|Map|Find (Variables|Orientation|GetOrientationYawOffsets) (Variables|Orientation|GetCardinalDirection)) "0, 0,1.000000") (Utilities|Map|Find (Variables|Orientation|GetOrientationLocationOffsets) (Variables|Orientation|GetCardinalDirection))) (Math|Rotator|MakeRotator 0.0 0.0 (+ (.yaw (Utilities|Array|Get(aref) BGARotationArray _index)) (Utilities|Map|Find (Variables|Orientation|GetOrientationYawOffsets) (Variables|Orientation|GetCardinalDirection))))) "Undefined" "SelectDefaultAtRuntime" 0 0))
        (bind _asbuilding_static_mesh (Utilities|Casting|CastToBuildingSMActor _returnvalue))
        (Class|BuildingActor|SetTeam _asbuilding_static_mesh (Team|GetActorTeam (Actor|GetOwner)))
        (Building|InitializeKismetSpawnedBuildingActor _asbuilding_static_mesh 0 0 true 0 false)
        (bind _asbuilding_trap (Utilities|Casting|CastToBuildingTrap _asbuilding_static_mesh)
          (:then
            (Trap|SpawnedfromItemDefinition _asbuilding_trap "/Game/Athena/Items/Traps/DamageTrap/TID_Floor_Spikes_Athena_R_T01.TID_Floor_Spikes_Athena_R_T01" 1))
          (:CastFailed
            _))
        (if (not (and (Building|IsWorldLocValid (Building|GetBuildingStructuralSupportSystem) (+ (Transformation|GetActorLocation _asbuilding_static_mesh) "0, 0,384.000000")) (Building|IsWorldLocValid (Building|GetBuildingStructuralSupportSystem) (+ (Transformation|GetActorLocation _asbuilding_static_mesh) "0, 0,-384.000000"))))
          (Actor|DestroyActor _asbuilding_static_mesh))))))

(fn HandleProps
  (params (Array :type Array%20of%20Actor%20Object%20References))
  (for _array_element Array
    (Utilities|Casting|CastToBuildingProp _array_element
      (:then
        (Utilities|Casting|CastToTiered_Chest_Athena _array_element
          (:then
            _)
          (:CastFailed
            (Utilities|Casting|CastToBuildingPropBigHealth _array_element
              (:then
                _)
              (:CastFailed
                (if (Variables|/Script/Engine.Actor|GetbCanBeDamaged _array_element)
                  (Utilities|Array|Add (Variables|Tower|GetBuildingActorsToDestroy) _array_element)
                  (else
                    _))))
            (Utilities|Array|Add (Variables|Tower|GetBuildingActorsToIgnore) _array_element))))
      (:CastFailed
        (Utilities|Casting|CastToBP_DamageBalloon_Athena _array_element
          (:then
            _)
          (:CastFailed
            (Utilities|Casting|CastToBuildingActor _array_element)))))))

(fn SetBaseDestination
  (params (Hit :type Hit%20Result%20Structure%20%28by%20ref%29))
  (bind (_bblockinghit _binitialoverlap _time _distance _location _impactpoint _normal _impactnormal) (Collision|BreakHitResult Hit))
  (Networking|SwitchHasAuthority
    (:Authority
      (Variables|BuildingAndClearing|SetBaseDestination (Fortnite|SnaptoFortBuildingGrid (+ _impactpoint (* _impactnormal 1)) (Math|Vector|MakeVector 0.0 0.0 (Variables|SnaptoGrid|GetGridZOffset)) (Math|Vector|MakeVector 0.0 (Variables|SnaptoGrid|GetGridYOffset)))))))

(fn SpawnTires
  (params (ReferenceLocation :type Vector))
  (bind _cardinaldirection (Variables|Orientation|GetCardinalDirection))
  (bind _orientationlocationoffsets (Variables|Orientation|GetOrientationLocationOffsets))
  (bind _value (Utilities|Map|Find _orientationlocationoffsets _cardinaldirection))
  (bind _cardinaldirection_2 (Variables|Orientation|GetCardinalDirection))
  (bind _orientationyawoffsets (Variables|Orientation|GetOrientationYawOffsets))
  (bind _value_2 (Utilities|Map|Find _orientationyawoffsets _cardinaldirection_2))
  (bind _bgalocations_tires (Variables|Tires|GetBGALocations_Tires))
  (bind _bgaclasses_tires (Variables|Tires|GetBGAClasses_Tires))
  (if (> (Utilities|Array|Length _bgaclasses_tires) 0)
    (if (Creative|Budget|CanPlaceClassInstances (Fortnite|GetFortPlayerControllerfromActor (Actor|GetOwner)) _bgaclasses_tires ReferenceLocation)
      (for _index (range (Utilities|Array|Length (Variables|Tires|GetBGAClasses_Tires)))
        (bind (_outhit _returnvalue) (Collision|SphereTraceByChannel (+ (Math|Vector|vector+vector ReferenceLocation (Math|Vector|RotateVectorAroundAxis (Utilities|Array|Get(aref) _bgalocations_tires _index) _value_2 "0, 0,1.000000") _value) "0, 0,384.000000") (Math|Vector|vector+vector ReferenceLocation (Math|Vector|RotateVectorAroundAxis (Utilities|Array|Get(aref) _bgalocations_tires _index) _value_2 "0, 0,1.000000") _value) 32.0 "TraceTypeQuery10" false (Variables|Tower|GetBuildingActorsToIgnore)))
        (Development|PrintString (Utilities|String|Append "Tower Grenade Tire trace hit: " (Utilities|GetDisplayName (.HitActor (Collision|BreakHitResult _outhit)))) (Variables|Default|GetIsDebug) (Variables|Default|GetIsDebug))
        (bind _returnvalue_2 (Game|SpawnActorfromClass (Utilities|Array|Get(acopy) (Variables|Tires|GetBGAClasses_Tires) _index) (Math|Transform|MakeTransform (select _returnvalue (- (.Location (Collision|BreakHitResult _outhit)) "0, 0,32.000000") (Math|Vector|vector+vector ReferenceLocation (Math|Vector|RotateVectorAroundAxis (Utilities|Array|Get(aref) _bgalocations_tires _index) _value_2 "0, 0,1.000000") _value)) (Math|Rotator|MakeRotator 0.0 0.0 (+ (.yaw (Utilities|Array|Get(aref) (Variables|Tires|GetBGARotations_Tires) _index)) _value_2))) "Undefined" "SelectDefaultAtRuntime" 0 0))
        (bind _asbuilding_actor (Utilities|Casting|CastToBuildingActor _returnvalue_2))
        (Building|InitializeKismetSpawnedBuildingActor _asbuilding_actor 0 0 true 0 false)
        (if (not (and (Building|IsWorldLocValid (Building|GetBuildingStructuralSupportSystem) (+ (Transformation|GetActorLocation _returnvalue_2) "0, 0,384.000000")) (Building|IsWorldLocValid (Building|GetBuildingStructuralSupportSystem) (+ (Transformation|GetActorLocation _returnvalue_2) "0, 0,-384.000000"))))
          (Actor|DestroyActor _returnvalue_2))))))

(fn ConstructionScript
  (Networking|CanExecuteCosmeticEvents
    (:True
      (Components|Movement|SetUpdatedComponent (Variables|Components|GetRotatingMovement) (Variables|Components|GetMesh)))))
